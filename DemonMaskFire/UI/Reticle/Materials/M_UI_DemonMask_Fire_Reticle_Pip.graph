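# Translucent UI material for a Demon Mask Fire reticle pip that shows a Progress value as a bottom-up fill.
# A FillMask step on vertical UV reveals grey over black, and floored Progress switches it to a tip-to-body gradient with radial accent when full.
# The center-scaled DemonFire texture's green channel acts as a distance field, antialiased into the pip shape over a shadow color.
# Opacity combines that shape with a soft fill-masked halo, keeping the unfilled part faint.
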Material /DemonMaskFire/UI/Reticle/Materials/M_UI_DemonMask_Fire_Reticle_Pip
graph {
 properties {bAutomaticallySetUsageInEditor=false, blendMode=BLEND_Translucent, materialDomain=MD_UI, usedShadingModels=Default Lit}
 nodes [
  n1 = Constant3Vector() {constant=<rgba 0.5 0.5 0.5 0>}
  n2 = Multiply(A=Subtract(A=MaterialFunctionCall() {materialFunction=&LinearGradient}.VGradient) {constB=0.35}) {constB=2.5}
  n3 = LinearInterpolate(
   A=LinearInterpolate(
    A=VectorParameter() {parameterName=PipTipColor, defaultValue=<rgba 0.8879 0.0802 0.0369 1>}.RGB
    B=VectorParameter() {parameterName=PipColor, defaultValue=<rgba 0.8879 0.3295 0.0777 1>}.RGB
    Alpha=Saturate(None=@n2)
   )
   B=Constant3Vector() {constant=<rgba 0.8879 0.2439 0 1>}
   Alpha=MaterialFunctionCall(Radius=ScalarParameter() {parameterName=AccentColorColor}) {materialFunction=&RadialGradientExponential}.RadialGradientExponential
  )
  n4 = Floor(None=ScalarParameter() {defaultValue=0.5, parameterName=Progress, sliderMax=1})
  n5 = TextureSample(UVs=MaterialFunctionCall(UVs=TextureCoordinate(), Texture Scale=Constant() {r=1.6}) {materialFunction=&ScaleUVsByCenter}.UVs) {samplerSource=SSM_Clamp_WorldGroupSettings, texture=/DemonMaskFire/UI/Reticle/Textures/T_UI_Reticle_DemonFire}
  n6 = MaterialFunctionCall(DistanceField=@n5.G, Threshold=Constant() {r=0.5}, Smoothing=Constant() {r=0.08}) {materialFunction=&MF_UI_AntiAliasing}
  LinearInterpolate(
   A=VectorParameter() {parameterName=Shadow Color, defaultValue=<rgba 0 0 0 1>}.RGB
   B=LinearInterpolate(A=Constant3Vector() {constant=<rgba 0 0 0 1>}, B=LinearInterpolate(A=@n1, B=@n3, Alpha=@n4), Alpha=NamedRerouteUsage())
   Alpha=@n6.Outer
  )
  n7 = Multiply(A=MaterialFunctionCall(DistanceField=@n5.G, Threshold=Constant() {r=0.45}, Smoothing=Constant() {r=0.9}) {materialFunction=&MF_UI_AntiAliasing}.Outer) {constB=0.4}
  Max(A=Multiply(A=@n6.Outer) {constB=0.3}, B=Multiply(A=Max(A=@n6.Outer, B=@n7), B=NamedRerouteUsage())) {constB=0.3}
  n8 = ComponentMask(None=MaterialFunctionCall() {materialFunction=&GetUserInterfaceUV}.Normalized UV) {g=true}
  NamedRerouteDeclaration(None=Step(Y=OneMinus(None=@n8), X=ScalarParameter() {defaultValue=0.5, parameterName=Progress, sliderMax=1})) {name=FillMask, nodeColor=<rgba 0 0 0 1>}
 ]
}
