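# Sound cue for the target-lock sound of the Demon Mask Fire player item, played as weapon foley at reduced volume.
# A Random node picks one of three equally weighted wave variations.
# Concurrency overrides cap it at two simultaneous instances, stopping the oldest, with a short retrigger time.
# Distance attenuation is disabled, though listener-focus settings are present in the attenuation overrides.
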SoundCue /DemonMaskFire/Assets/Audio/DemonMaskFire_Player/SC_Mask_Demon_Fire_Target_Lock
root=Random_0
nodes [
 {id=Random_0, children [WavePlayer_1,WavePlayer_2,WavePlayer_3], properties {weights [1,1,1]}}
 {id=WavePlayer_1, properties {soundWaveAssetPtr=&/Mask_Demon_Fire_Target_Lock_01}}
 {id=WavePlayer_2, properties {soundWaveAssetPtr=&/Mask_Demon_Fire_Target_Lock_02}}
 {id=WavePlayer_3, properties {soundWaveAssetPtr=&/Mask_Demon_Fire_Target_Lock_03}}
]
properties {
 bOverrideConcurrency=true
 maxDistance=2097152
 soundClassObject=&Fort_WeaponFoley
 volumeMultiplier=0.35
 attenuationOverrides {
  bAttenuate=false
  bEnableFocusInterpolation=true
  bEnableListenerFocus=true
  falloffDistance=1200
  nonFocusVolumeAttenuation=0.5
  stereoSpread=10
  attenuationShapeExtents {x=0}
 }
 concurrencyOverrides {maxCount=2, resolutionRule=StopOldest, retriggerTime=0.2}
}
