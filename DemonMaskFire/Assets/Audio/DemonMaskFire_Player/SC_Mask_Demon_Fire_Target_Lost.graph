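# Sound cue for the Demon Mask Fire player audio, named for the moment its target is lost.
# A Random node picks one of three equally weighted wave variations.
# Output goes to the weapon foley sound class at reduced volume.
# A concurrency override caps simultaneous instances at a small count, stopping the oldest, with a short retrigger window.
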SoundCue /DemonMaskFire/Assets/Audio/DemonMaskFire_Player/SC_Mask_Demon_Fire_Target_Lost
root=Random_1
nodes [
 {id=Random_1, children [WavePlayer_0,WavePlayer_4,WavePlayer_5], properties {weights [1,1,1]}}
 {id=WavePlayer_0, properties {soundWaveAssetPtr=&/Mask_Demon_Fire_Target_Lost_01}}
 {id=WavePlayer_4, properties {soundWaveAssetPtr=&/Mask_Demon_Fire_Target_Lost_02}}
 {id=WavePlayer_5, properties {soundWaveAssetPtr=&/Mask_Demon_Fire_Target_Lost_03}}
]
properties {
 bOverrideConcurrency=true
 maxDistance=2097152
 soundClassObject=&Fort_WeaponFoley
 volumeMultiplier=0.7
 attenuationOverrides {
  bAttenuate=false
  bEnableFocusInterpolation=true
  bEnableListenerFocus=true
  falloffDistance=1200
  nonFocusVolumeAttenuation=0.5
  stereoSpread=10
  attenuationShapeExtents {x=0}
 }
 concurrencyOverrides {maxCount=2, resolutionRule=StopOldest, retriggerTime=0.2}
}
