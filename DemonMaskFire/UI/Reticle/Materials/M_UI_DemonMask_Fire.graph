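# UI material for the Demon Mask Fire reticle, a translucent crosshair with a center dot.
# A mirrored distance-field texture shifts with the isFire and isADS parameters, and isEmpty lowers its opacity.
# Antialiasing functions threshold the texture and a procedural circle twice, layering crisp fills over soft, faint shadows.
# An animatable progress gradient is built from rotated UVs, alongside a white-to-red color blend.
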Material /DemonMaskFire/UI/Reticle/Materials/M_UI_DemonMask_Fire
graph {
 properties {bAutomaticallySetUsageInEditor=false, blendMode=BLEND_Translucent, materialDomain=MD_UI, usedShadingModels=Default Lit}
 nodes [
  n1 = ScalarParameter() {defaultValue=0.204, parameterName=Base_Offset}
  n2 = LinearInterpolate(B=ScalarParameter() {defaultValue=0.3, parameterName=Fire_Offset_Percent}, Alpha=ScalarParameter() {parameterName=isFire})
  n3 = Add(
   A=Add(A=@n1, B=Multiply(A=@n1, B=@n2))
   B=LinearInterpolate(B=ScalarParameter() {defaultValue=-0.1, parameterName=ADS_Offset}, Alpha=ScalarParameter() {parameterName=isADS})
  )
  n4 = Constant()
  n5 = TextureSample(UVs=MaterialFunctionCall!(UV, Offset=AppendVector(A=@n3, B=@n4), Rotation, Scale=Constant2Vector() {g=0.5, r=1}, MirrorY) {materialFunction=/FortUILibrary/MaterialFunctions/Basics/MF_UI_Symmetry_2x1_DEPRECATED}.Result) {samplerSource=SSM_Clamp_WorldGroupSettings, texture=/DemonMaskFire/UI/Reticle/Textures/T_UI_Reticle_DemonFire}
  n6 = NamedRerouteDeclaration(None=MaterialFunctionCall(DistanceField=@n5.R, Threshold=Constant() {r=0.45}, Smoothing=Constant() {r=0.05}) {materialFunction=&MF_UI_AntiAliasing}.Outer) {name=Fill, nodeColor=<rgba 0 0 0 1>}
  n7 = Multiply(A=MaterialFunctionCall(DistanceField=@n5.R, Threshold=Constant() {r=0.45}, Smoothing=Constant() {r=0.8}) {materialFunction=&MF_UI_AntiAliasing}.Outer) {constB=0.4}
  n8 = ScalarParameter() {defaultValue=0.8, parameterName=Base_Opacity}
  n9 = ScalarParameter() {defaultValue=0.3, parameterName=Empty_Opacity}
  n10 = ScalarParameter() {parameterName=isEmpty}
  Add(A=Multiply(A=Max(A=@n6, B=@n7), B=LinearInterpolate(A=@n8, B=@n9, Alpha=@n10)), B=NamedRerouteUsage()) {constB=0}
  n11 = Constant3Vector() {constant=<rgba 0.8879 0.0802 0.0369 1>}
  LinearInterpolate(A=LinearInterpolate(A=Constant3Vector() {constant=<rgba 0.901 0.901 0.901 0>}, B=@n11, Alpha=NamedRerouteUsage()), Alpha=NamedRerouteUsage())
  n12 = TextureCoordinate()
  n13 = Constant() {r=0.2}
  n14 = OneMinus(None=MaterialFunctionCall(
    UVs=MaterialFunctionCall(UVs=@n12, Texture Scale=@n13) {materialFunction=&ScaleUVsByCenter}.UVs
    Size=Constant() {r=0.1}
   ) {materialFunction=&MF_UI_Circle}.SDF)
  n15 = NamedRerouteDeclaration(None=MaterialFunctionCall(DistanceField=@n14, Threshold=Constant() {r=1}) {materialFunction=&MF_UI_AntiAliasing}.Outer) {name=Dot, nodeColor=<rgba 0 0 0 1>}
  n16 = Multiply(A=MaterialFunctionCall(DistanceField=@n14, Threshold=Constant() {r=1}, Smoothing=Constant() {r=0.1}) {materialFunction=&MF_UI_AntiAliasing}.Outer) {constB=0.4}
  NamedRerouteDeclaration(None=Multiply(A=Max(A=@n15, B=@n16), B=ScalarParameter() {defaultValue=1, parameterName=DotOpacity}) {constB=0}) {name=DotShadow, nodeColor=<rgba 0 0 0 1>}
  n17 = ScalarParameter() {group=Progress_Layout, parameterName=ProgressEdgeSharpness, sliderMax=1, sortPriority=2}
  n18 = MaterialFunctionCall(UVs=TextureCoordinate(), "Rotation Angle (0-1)"=Constant() {r=0.5}) {materialFunction=&CustomRotator}
  n19 = ScalarParameter() {
   defaultValue=0.28
   desc="When progress value changes, it defines where change starts.\r\n0 - 1"
   group=Progress
   parameterName=ProgressStart
   sliderMax=1
   sortPriority=1
  }
  n20 = ScalarParameter() {
   defaultValue=0.72
   desc="When progress value changes, it defines where change ends.\r\n0 - 1"
   group=Progress
   parameterName=ProgressEnd
   sliderMax=1
   sortPriority=2
  }
  n21 = ScalarParameter() {
   defaultValue=1
   desc="Property for animating progress value change (from ProgressStart to ProgressEnd)\r\n0 - 1"
   group=Progress
   parameterName=AnimateProgress
   sliderMax=1
   sortPriority=3
  }
  n22 = LinearInterpolate(A=@n19, B=@n20, Alpha=@n21)
  n23 = ComponentMask(None=Add(A=Multiply(A=@n18.Rotated Values) {constB=-1}, B=NamedRerouteDeclaration(None=@n22) {name=Progress, nodeColor=<rgba 0 0 0 1>}) {constB=0.5}) {g=true}
  NamedRerouteDeclaration(None=Saturate(None=SmoothStep(Max=@n17, Value=@n23) {constMax=0.01})) {name=ProgressGraident, nodeColor=<rgba 0 0 0 1>}
  Reroute(None=@n11)
 ]
}
