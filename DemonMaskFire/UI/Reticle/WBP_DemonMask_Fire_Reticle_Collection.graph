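# Reticle widget for the fire Demon Mask that shows recharge stack pips, built as a child of the shared recharge-stack pip collection base.
# A fixed-size SizeBox holds a bottom-centered vertical DynamicEntryBox of fire pip widgets, overlapped through negative spacing.
# On Construct it hands that entry box to the base via SetDynamicEntryBox.
# Its GenerateStacks override only forwards total and current stack counts to the parent.
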WidgetBlueprint /DemonMaskFire/UI/Reticle/WBP_DemonMask_Fire_Reticle_Collection
graphs [
 EventGraph {nodes [
   Event Construct()
   > SetDynamicEntryBox!(
    execute
    self:Fort Weapon Reticle Recharge Stack Base Pip Collection Base Object Reference
    DynamicEntryBox:Dynamic Entry Box Object Reference=Get DynamicEntryBox_Stacks() -> (DynamicEntryBox_Stacks:Dynamic Entry Box Object Reference)
   ) -> (then)
  ]}
 GenerateStacks {nodes [
   n1 = GenerateStacks!() -> (then, TotalStacks:int=0, CurrentStacks:int=0) {is=title K2Node_FunctionEntry}
   > Parent: GenerateStacks!(execute, TotalStacks:int=@n1, CurrentStacks:int=@n1.CurrentStacks) -> (then) {is=title K2Node_CallParentFunction}
  ]}
]
variables []
widgets {
 info {
  inheritedWidgetCount=0
  parentClass=/Game/Athena/HUD/reticles/ReticleExtensionWidgets/ChargeStack/WBP_Reticle_RechargeStackBase_PipCollection.WBP_Reticle_RechargeStackBase_PipCollection_C
  rootWidgetClass=&SizeBox
  widgetCount=2
 }
 widgets [
  SizeBox_0 {
   widgetClassPath=&SizeBox
   widget=SizeBox_0
   properties {bOverride_HeightOverride=true, bOverride_WidthOverride=true, heightOverride=190, widthOverride=190}
  }
  DynamicEntryBox_Stacks {
   widgetClassPath=&DynamicEntryBox
   bIsVariable=true
   parent=SizeBox_0
   widget=DynamicEntryBox_Stacks
   properties {
    entryBoxType=Vertical
    entryWidgetClass=&/WBP_DemonMask_Fire_Reticle_Pip.WBP_DemonMask_Fire_Reticle_Pip_C
    numDesignerPreviewEntries=2
    entrySpacing=<0 -45>
    radialBoxSettings {angleBetweenItems=21.12, bClockwiseOrder=false, bDistributeItemsEvenly=false, marginSize=0, sectorCentralAngle=360, startingAngle=260}
   }
   slot {class=SizeBoxSlot, properties {horizontalAlignment=HAlign_Center, verticalAlignment=VAlign_Bottom}}
  }
 ]
}
