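# Translucent UI material for the Demon Mask Fire reticle's attachment point, drawing fang and mouth shapes from a channel-packed lock-on texture.
# Fangs sample the green channel through offset, 2x2-symmetry UVs.
# The mouth samples red, switching between flipped and unflipped offset coordinates per vertical half, then scaling about center.
# Distance-field functions turn both channels into fills, which are weighted by separate opacity parameters and summed.
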Material /DemonMaskFire/UI/Reticle/Materials/M_UI_DemonMask_Fire_AttachmentPoint
graph {
 properties {bAutomaticallySetUsageInEditor=false, blendMode=BLEND_Translucent, materialDomain=MD_UI, usedShadingModels=Default Lit}
 nodes [
  n1 = ScalarParameter() {defaultValue=0.05, parameterName=OffsetX}
  n2 = ScalarParameter() {defaultValue=0.052, parameterName=OffsetY}
  n3 = TextureSample(UVs=Add(
    A=MaterialFunctionCall!(UV, Swizzle UV) {materialFunction=/FortUILibrary/MaterialFunctions/Basics/MF_UI_Symmetry_2x2_New}.Symmetry
    B=AppendVector(A=@n1, B=@n2)
   )) {samplerSource=SSM_Clamp_WorldGroupSettings, texture=/DemonMaskFire/UI/Reticle/Textures/T_UI_Reticle_DemonFire_lockon}
  n4 = Constant() {r=0.2}
  n5 = Constant() {r=0.9}
  Add(
   A=Multiply(
    A=MaterialFunctionCall!(RGBA Channels=@n3.G, Glow Min=@n4, Glow Max=@n5, Outline Thickness) {materialFunction=/FortUILibrary/MaterialFunctions/Basics/MF_UI_ChannelPackedDistanceField}.Fill
    B=ScalarParameter() {parameterName=Fangs Opacity}
   )
   B=Multiply(A=NamedRerouteUsage(), B=ScalarParameter() {defaultValue=0.7, parameterName=Mouth Opacity})
  )
  Constant3Vector() {constant=<rgba 1 1 1 0>}
  n6 = Multiply(A=TextureCoordinate(), B=Constant2Vector() {g=2, r=1})
  n7 = AppendVector(A=ScalarParameter() {parameterName=MX_Offset}, B=ScalarParameter() {parameterName=MY_Offset})
  n8 = Add(A=Frac(None=@n6), B=@n7)
  n9 = MaterialFunctionCall(Float2=Add(A=OneMinus(None=@n8), B=Multiply(A=@n7) {constB=2})) {materialFunction=&BreakOutFloat2Components}
  n10 = MaterialFunctionCall(Float2=@n8) {materialFunction=&BreakOutFloat2Components}
  n11 = Saturate(None=Floor(None=ComponentMask(None=@n6) {g=true}))
  n12 = MaterialFunctionCall(
   UVs=LinearInterpolate(A=AppendVector(A=@n9.R, B=@n9.R), B=AppendVector(A=@n10.R, B=@n10.R), Alpha=OneMinus(None=@n11))
   Texture Scale=Constant2Vector() {g=1, r=0.6}
  ) {materialFunction=&ScaleUVsByCenter}
  NamedRerouteDeclaration(None=MaterialFunctionCall(
    RGBA Channels=TextureSample(UVs=@n12.UVs) {samplerSource=SSM_Clamp_WorldGroupSettings, texture=&T_UI_Reticle_DemonFire_lockon}.R
    Glow Min=Constant() {r=0.2}
    Glow Max=Constant() {r=0.9}
   ) {materialFunction=&MF_UI_ChannelPackedDistanceField}.Fill) {name=Mouth, nodeColor=<rgba 0 0 0 1>}
  Panner(Coordinate=TextureCoordinate()) {speedY=0.592}
 ]
}
