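# Fortnite weapon attachment-point reticle widget for Demon Mask Fire, showing whether a valid target is held.
# SetPointOfAttachmentState stops running animations, then plays the no-target or enemy-targeted animation from a boolean.
# Animations drive opacity, fang, mouth and offset parameters on a single image material.
# Sequencer event functions start and stop an endless locked-on pulse, and Destruct stops all animations.
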WidgetBlueprint /DemonMaskFire/UI/Reticle/WBP_Reticle_DemonMaskFire_AttachmentPoint
animations [
 Anim_NoTarget {
  binding_count=1
  length_seconds=0.3
  object=&self:Anim_NoTarget
  sequence {
   bindings [Image_ReticlePointOfAttachment_1 {
     id=#1
     possessed_class=Image
     tracks [{
       type=MovieSceneWidgetMaterialTrack
       sections [{
         type=MovieSceneParameterSection
         end_frame=6
         start_frame=0
         channels [
          Fangs Opacity {type=MovieSceneScriptingFloatChannel, const [0,1]}
          Opacity {type=MovieSceneScriptingFloatChannel, const [0.6,1]}
         ]
        }]
      }]
    }]
   tracks [{
     type=MovieSceneEventTrack
     sections [{
       type=MovieSceneEventTriggerSection
       channels [None {type=MovieSceneScriptingEventChannel, const ["<Struct 'MovieSceneEvent' (0x466f61ed200) {}>", 1]}]
      }]
    }]
  }
 }
 Anim_LockedOnTarget_Loop {
  binding_count=1
  length_seconds=0.5
  object=&self:Anim_LockedOnTarget_Loop
  sequence {bindings [Image_ReticlePointOfAttachment_1 {
     id=#2
     possessed_class=Image
     tracks [{
       type=MovieSceneWidgetMaterialTrack
       sections [{
         type=MovieSceneParameterSection
         end_frame=10
         start_frame=0
         channels [
          MY_Offset {type=MovieSceneScriptingFloatChannel, keys [[0,0],[4,0],[10,-0.17]]}
          Opacity {type=MovieSceneScriptingFloatChannel, keys [[0,0],[4,1],[10,0]]}
          Fangs Opacity {type=MovieSceneScriptingFloatChannel, const [1,1]}
          Mouth Opacity {type=MovieSceneScriptingFloatChannel, keys [[0,0.7],[10,0]]}
         ]
        }]
      }]
    }]}
 }
 Anim_EnemyTargetted {
  binding_count=1
  length_seconds=0.5
  object=&self:Anim_EnemyTargetted
  sequence {
   bindings [Image_ReticlePointOfAttachment_1 {
     id=#3
     possessed_class=Image
     tracks [{
       type=MovieSceneWidgetMaterialTrack
       sections [{
         type=MovieSceneParameterSection
         end_frame=10
         start_frame=0
         channels [
          OffsetX {type=MovieSceneScriptingFloatChannel, keys [[3,0.17],[7,0.03],[8,0.05]]}
          OffsetY {type=MovieSceneScriptingFloatChannel, keys [[3,0.192],[7,0.032],[8,0.052]]}
          Fangs Opacity {type=MovieSceneScriptingFloatChannel, keys [[3,0],[7,1]]}
          Opacity {type=MovieSceneScriptingFloatChannel, const [1,1]}
          MY_Offset {type=MovieSceneScriptingFloatChannel, keys [[0,0.18],[3,0]]}
          Mouth Opacity {type=MovieSceneScriptingFloatChannel, keys [[0,0],[3,0.7]]}
         ]
        }]
      }]
    }]
   tracks [{
     type=MovieSceneEventTrack
     sections [{
       type=MovieSceneEventTriggerSection
       channels [None {type=MovieSceneScriptingEventChannel, const ["<Struct 'MovieSceneEvent' (0x466f61ee240) {}>", 1]}]
      }]
    }]
  }
 }
]
graphs [
 EventGraph {nodes [
   Event Destruct()
   > Parent: Destruct()
   > StopAllAnimations(self:User Widget Object Reference)
  ]}
 Sequencer Events {nodes [
   StartLockedOnLoop() {is=title K2Node_CustomEvent}
   n1 = Get Anim_LockedOnTarget_Loop() -> (Anim_LockedOnTarget_Loop:Widget Animation Object Reference)
   > PlayAnimation(InAnimation=@n1, NumLoopsToPlay=0)
   StopLockedOnLoop() {is=title K2Node_CustomEvent}
   > StopAnimation(InAnimation=@n1)
  ]}
 SetPointOfAttachmentState {nodes [
   n1 = SetPointOfAttachmentState!() -> (then, HasValidTarget:bool=~) {is=title K2Node_FunctionEntry}
   > StopAllAnimations()
   n2 = Select(
    Option 0:Widget Animation Object Reference=Get Anim_NoTarget() -> (Anim_NoTarget:Widget Animation Object Reference)
    Option 1:Widget Animation Object Reference=Get Anim_EnemyTargetted() -> (Anim_EnemyTargetted:Widget Animation Object Reference)
    Index=@n1
   ) -> (ReturnValue:Widget Animation Object Reference)
   > Is Valid(InputObject=@n2)
   > PlayAnimation(InAnimation=@n2)
   > n3 = Set PreviousReticleState(PreviousReticleState:DemonMaskFire_ReticleState Enum=NewEnumerator0) -> (Output_Get:DemonMaskFire_ReticleState Enum=NewEnumerator0)
   > n4 = Switch on DemonMaskFire_ReticleState!(execute, Selection:DemonMaskFire_ReticleState Enum=@n3) -> (
    NewEnumerator0:Exec
    NewEnumerator1:Exec
    NewEnumerator7:Exec
    NewEnumerator2:Exec
    NewEnumerator3:Exec
    NewEnumerator4:Exec
    NewEnumerator5:Exec
    NewEnumerator6:Exec
    NewEnumerator10:Exec
   ) {is=title K2Node_SwitchEnum}
   SetRenderTransform!(execute=[@n4.NewEnumerator0, @n4.NewEnumerator7], self:Widget Object Reference, InTransform:Widget Transform Structure) -> (then)
  ]}
]
variables [PreviousReticleState {type=DemonMaskFire_ReticleState, default=NoTarget}]
widgets {
 info {inheritedWidgetCount=1, parentClass=/Script/FortniteUI.FortWeaponAttachmentPointWidget, rootWidgetClass=&Overlay, widgetCount=3}
 widgets [
  Overlay_LockedOn {
   widgetClassPath=&Overlay
   bIsVariable=true
   widget=Overlay_LockedOn
   properties {renderTransform {angle=0, scale=<0.7 0.7>, shear=<0 0>, translation=<0 0>}}
  }
  Image_ReticlePointOfAttachment_1 {
   widgetClassPath=&Image
   bIsVariable=true
   parent=Overlay_LockedOn
   widget=Image_ReticlePointOfAttachment_1
   properties {
    brush {
     bIsDynamicallyLoaded=false
     drawAs=Image
     imageType=FullColor
     mirroring=NoMirror
     resourceName=None
     resourceObject=/DemonMaskFire/UI/Reticle/Materials/M_UI_DemonMask_Fire_AttachmentPoint
     tiling=NoTile
     imageSize=<45 45>
     margin=<ltrb 0 0 0 0>
     outlineSettings {
      bUseBrushTransparency=false
      roundingType=HalfHeightRadius
      width=0
      color {colorUseRule=UseColor_Specified, specifiedColor=<rgba 0 0 0 0>}
      cornerRadii=<0 0 0 0>
     }
     tintColor {colorUseRule=UseColor_Specified, specifiedColor=<rgba 1 1 1 1>}
     uVRegion {bIsValid=false, max=<0 0>, min=<0 0>}
    }
    renderTransform {angle=0, scale=<1.5 1.5>, shear=<0 0>, translation=<0 0>}
   }
   slot {class=OverlaySlot, properties {horizontalAlignment=HAlign_Fill, verticalAlignment=VAlign_Fill}}
  }
  TutorialHighlight {widgetClassPath=&TutorialHighlight}
 ]
}
