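# Reticle widget for the Demon Mask Fire weapon, built on the recharge-stack reticle base.
# Weapon and target-state events drive material animations for firing, aiming, empty ammo and enemy in range, plus target lock sounds and attachment-point state.
# Ammo changes update a material progress bar and toggle a 0.1-second cooldown poll that feeds recharge pips while the magazine is below full.
# Counters limit range-triggered animations to state transitions.
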WidgetBlueprint /DemonMaskFire/UI/Reticle/WBP_DemonMask_Fire_Reticle
animations [
 Anim_EnemyInRange {
  binding_count=1
  length_seconds=0.25
  object=&self:Anim_EnemyInRange
  sequence {bindings [Reticle {
     id=#1
     possessed_class=Image
     tracks [{
       type=MovieSceneWidgetMaterialTrack
       sections [{
         type=MovieSceneParameterSection
         end_frame=5
         start_frame=0
         channels [Base_Opacity {type=MovieSceneScriptingFloatChannel, keys [[0,0.6],[5,1]]}]
        }]
      }]
    }]}
 }
 Anim_ADS {
  binding_count=1
  length_seconds=0.25
  object=&self:Anim_ADS
  sequence {bindings [Reticle {
     id=#2
     possessed_class=Image
     tracks [{
       type=MovieSceneWidgetMaterialTrack
       sections [{type=MovieSceneParameterSection, end_frame=5, start_frame=0, channels [isADS {type=MovieSceneScriptingFloatChannel, keys [[0,0],[5,1]]}]}]
      }]
    }]}
 }
 Anim_NoAmmo {
  binding_count=1
  length_seconds=0.25
  object=&self:Anim_NoAmmo
  sequence {bindings [Reticle {
     id=#3
     possessed_class=Image
     tracks [{
       type=MovieSceneWidgetMaterialTrack
       sections [{type=MovieSceneParameterSection, end_frame=5, start_frame=0, channels [isEmpty {type=MovieSceneScriptingFloatChannel, keys [[0,0],[5,1]]}]}]
      }]
    }]}
 }
 Anim_Fire {
  binding_count=1
  length_seconds=0.5
  object=&self:Anim_Fire
  sequence {bindings [Reticle {
     id=#4
     possessed_class=Image
     tracks [{
       type=MovieSceneWidgetMaterialTrack
       sections [{
         type=MovieSceneParameterSection
         end_frame=10
         start_frame=0
         channels [
          Opacity {type=MovieSceneScriptingFloatChannel, thinned_from=5, keys [[0,0.8],[3.0003,1],[10.0003,0.8]]}
          isFire {type=MovieSceneScriptingFloatChannel, keys [[0,0],[3,1.4],[10,0]]}
         ]
        }]
      }]
    }]}
 }
 Anim_Equip {
  binding_count=2
  length_seconds=0.1711
  object=&self:Anim_Equip
  sequence {bindings [
    WBP_DemonMask_Fire_Reticle_Collection {
     id=#5
     possessed_class=WBP_DemonMask_Fire_Reticle_Collection_C
     tracks [{
       type=MovieSceneColorTrack
       display_name=Color and Opacity
       sections [{
         type=MovieSceneColorSection
         end_frame=3
         start_frame=0
         channels [
          Color.R {type=MovieSceneScriptingFloatChannel, const [1,1]}
          Color.G {type=MovieSceneScriptingFloatChannel, const [1,1]}
          Color.B {type=MovieSceneScriptingFloatChannel, const [1,1]}
          Color.A {type=MovieSceneScriptingFloatChannel, keys [[0,0],[3,1]]}
         ]
        }]
      }]
    }
    Reticle {
     id=#6
     possessed_class=Image
     tracks [{
       type=MovieSceneWidgetMaterialTrack
       sections [{
         type=MovieSceneParameterSection
         end_frame=3
         start_frame=0
         channels [
          Base_Opacity {type=MovieSceneScriptingFloatChannel, keys [[0,0],[3,0.6]]}
          isEmpty {type=MovieSceneScriptingFloatChannel, const [0,1]}
          Saturation {type=MovieSceneScriptingFloatChannel, const [0,1]}
         ]
        }]
      }]
    }
   ]}
 }
]
graphs [
 BindToWeaponEvents {nodes [
   n1 = BindToWeaponEvents!() -> (then, DoBind:bool, EquippedWeapon:Fort Weapon Object Reference) {is=title K2Node_FunctionEntry}
   > Parent: BindToWeaponEvents!(execute, DoBind:bool=@n1, EquippedWeapon:Fort Weapon Object Reference=@n1.EquippedWeapon) -> (then) {is=title K2Node_CallParentFunction}
   > n2 = Cast To B_DemonMask_Fire!(execute, Object=@n1.EquippedWeapon) -> (then, CastFailed, AsB Demon Mask Fire:B Demon Mask Fire Object Reference) {is=title K2Node_DynamicCast}
   > Set As B Demon Mask Fire(As B Demon Mask Fire:B Demon Mask Fire Object Reference=@n2) -> (Output_Get:B Demon Mask Fire Object Reference)
   > Bind Event to On Release Trigger(
    self:Fort Weapon Object Reference=Get As B Demon Mask Fire() -> (As B Demon Mask Fire:B Demon Mask Fire Object Reference)
    Delegate=Create Event()
   ) {is=title K2Node_AddDelegate}
   > Bind Event to Update Reticle State(self:B Demon Mask Fire Object Reference=Get As B Demon Mask Fire(), Delegate=Create Event()) {is=title K2Node_AddDelegate}
  ]}
 EventGraph {nodes [
   OnWeaponFired!() -> (OutputDelegate, then, Weapon:Fort Weapon Object Reference) {is=title K2Node_CustomEvent}
   n1 = Get As B Demon Mask Fire()
   > Branch(Condition="integer > integer"(A=GetLocalMagazineAmmoCount!(self:Fort Weapon Object Reference=@n1) -> (ReturnValue:int=0)))
   > PlayAnimationForward(InAnimation=Get Anim_Fire() -> (Anim_Fire:Widget Animation Object Reference))
   n2 = OnValidTargetChanged!() -> (OutputDelegate, then, bIsValidTarget:bool=~) {is=title K2Node_CustomEvent}
   > SetPointOfAttachmentState!(
    execute
    self:WBP Reticle Demon Mask Fire Attachment Point Object Reference=Get WBP_Reticle_DemonMaskFire_AttachmentPoint() -> (WBP_Reticle_DemonMaskFire_AttachmentPoint:WBP Reticle Demon Mask Fire Attachment Point Object Reference)
    HasValidTarget:bool=@n2.bIsValidTarget
   ) -> (then)
   > n3 = Branch(Condition=@n2.bIsValidTarget)
   n4 = Get Anim_EnemyInRange() -> (Anim_EnemyInRange:Widget Animation Object Reference)
   > PlayAnimationForward(InAnimation=@n4)
   > PlaySound2D(Sound=/DemonMaskFire/Assets/Audio/DemonMaskFire_Player/SC_Mask_Demon_Fire_Target_Lock)
   PlayAnimationReverse(execute=@n3.else, InAnimation=@n4)
   > PlaySound2D(Sound=/DemonMaskFire/Assets/Audio/DemonMaskFire_Player/SC_Mask_Demon_Fire_Target_Lost)
   n5 = Event OnLocalAmmoChanged!() -> (OutputDelegate, then, LocalCount:int=0, LocalRemaining:int=0) {is=title K2Node_Event}
   > Parent: OnLocalAmmoChanged!(execute, LocalCount:int=@n5.LocalCount, LocalRemaining:int=@n5.LocalRemaining) -> (then) {is=title K2Node_CallParentFunction}
   > UpdateCooldownQueryTimer(self:Fort Weapon Reticle Extension Widget Base Object Reference)
   > SetAmmoProgressBar()
   > n6 = Branch(Condition="integer < integer"(A=@n5.LocalCount, B=1))
   > PlayAnimationForward(InAnimation=Get Anim_NoAmmo() -> (Anim_NoAmmo:Widget Animation Object Reference))
   > Set isNoAmmoShowing(isNoAmmoShowing:bool=true) -> (Output_Get:bool)
   Branch(execute=@n6.else, Condition="Equal (Integer)"(A=@n5.LocalCount, B=1))
   > Branch(Condition=Get isNoAmmoShowing() -> (isNoAmmoShowing:bool))
   > PlayAnimationReverse(InAnimation=Get Anim_NoAmmo())
   > Set isNoAmmoShowing(isNoAmmoShowing=false)
   Event Construct()
   > Parent: Construct()
   > UpdateCooldownQueryTimer()
   > n7 = Set Target State Component(Target State Component:Fort Weapon Component Target State Object Reference=GetComponentByClass!(
     self:Actor Object Reference=Get As B Demon Mask Fire()
     ComponentClass:Actor Component Class Reference=/Script/FortniteGame.FortWeaponComponent_TargetState
    ) -> (ReturnValue:Fort Weapon Component Target State Object Reference)) -> (Output_Get:Fort Weapon Component Target State Object Reference)
   > Is Valid(InputObject=@n7)
   > Bind Event to On Is Valid Target Changed(self:Fort Weapon Component Target State Object Reference=@n7, Delegate=Create Event()) {is=title K2Node_AddDelegate}
   > PlayAnimation(InAnimation=Get Anim_Equip() -> (Anim_Equip:Widget Animation Object Reference))
   > Set isNoAmmoShowing(isNoAmmoShowing=false)
   n8 = Event OnTargetingChanged!() -> (OutputDelegate, then, bIsADS:bool) {is=title K2Node_Event}
   > n9 = Branch(Condition=@n8.bIsADS)
   n10 = Get Anim_ADS() -> (Anim_ADS:Widget Animation Object Reference)
   > PlayAnimationForward(InAnimation=@n10)
   PlayAnimationReverse(execute=@n9.else, InAnimation=@n10)
   Event Destruct()
   n11 = Get As B Demon Mask Fire()
   > Unbind all Events from Update Reticle State!(execute:Exec, self:B Demon Mask Fire Object Reference=@n11) -> (then:Exec) {is=title K2Node_ClearDelegate}
   > BindToWeaponEvents!(execute, self:WBP Reticle Recharge Stack Base Object Reference, DoBind:bool=false, EquippedWeapon:Fort Weapon Object Reference=@n11) -> (then) {is=title K2Node_CallFunction}
   > HandleCooldownTimer!(execute, self:Fort Weapon Reticle Extension Widget Base Object Reference, bActivateTimer:bool=false, TimerInterval:float=0) -> (then)
   Event UpdateRemainingCooldown() {is=title K2Node_Event}
   > n12 = GetRemainingCooldownTimer!(execute, self) -> (then, RemainingTime:double, TotalDuration:double)
   > UpdateRemainingRechargeTime!(
    execute
    self:Fort Weapon Reticle Recharge Stack Base Pip Collection Base Object Reference=GetPipCollection!(self:WBP Reticle Recharge Stack Base Object Reference) -> (PipCollection:WBP Reticle Recharge Stack Base Pip Collection Object Reference)
    TimeRemaining:float=@n12
    TotalDuration:float=@n12.TotalDuration
    PipPercentMatParamName:Name=Progress
    ShouldSetPipColor:bool=false
    NotReadyColor:Linear Color Structure
    ReadyColor:Linear Color Structure
    PipColorMatParamName:Name=None
   ) -> (then)
   Event UpdateCooldownQueryTimer() {is=title K2Node_Event}
   > n13 = Get!(execute:Exec) -> (then:Exec, else:Exec, As B Demon Mask Fire:B Demon Mask Fire Object Reference)
   > n14 = Branch(Condition="integer >= integer"(
     A=GetLocalMagazineAmmoCount(self=@n13) -> (ReturnValue=0)
     B=GetBulletsPerClip!(self:Fort Weapon Object Reference=@n13) -> (ReturnValue:int=0)
    ))
   > HandleCooldownTimer(bActivateTimer=false, TimerInterval=0)
   HandleCooldownTimer(execute=@n14.else, bActivateTimer=true, TimerInterval=0.1)
   n15 = UpdateReticelState_Event!() -> (OutputDelegate, then, PawnInRange:bool=~, Env Target In Range:bool=~) {is=title K2Node_CustomEvent}
   Set in Range State!(execute, self, bPlayerInRange:bool=@n15.PawnInRange, bEnvInRange:bool=@n15.Env Target In Range) -> (then) {is=title K2Node_CallFunction}
  ]}
 GetPipCollection {nodes [
   GetPipCollection() {is=title K2Node_FunctionEntry}
   > Return Node!(
    execute
    PipCollection:WBP Reticle Recharge Stack Base Pip Collection Object Reference=Get WBP_DemonMask_Fire_Reticle_Collection() -> (WBP_DemonMask_Fire_Reticle_Collection:WBP Demon Mask Fire Reticle Collection Object Reference)
   )
   Parent: GetPipCollection!() -> (PipCollection:WBP Reticle Recharge Stack Base Pip Collection Object Reference) {is=title K2Node_CallParentFunction}
  ]}
 GetRemainingCooldownTimer {nodes [
   GetRemainingCooldownTimer() {is=title K2Node_FunctionEntry}
   > n1 = Cast To FortPlayerController(Object=GetOwningPlayer())
   > n2 = Get()
   n3 = GetInventoryItem!(self:Fort Weapon Object Reference=@n2) -> (ReturnValue:Fort Item Object Reference)
   > n4 = Is Valid(InputObject=@n3)
   Return Node!(execute=[@n4.Is Not Valid, @n2.else, @n1.CastFailed], RemainingTime:double=0, TotalDuration:double=0)
   n5 = GetCoolDownTimeRemaining!(self:Fort Item Object Reference=@n3, FortPC:Fort Player Controller Object Reference=@n1) -> (CooldownDuration:float, ReturnValue:float)
   > Return Node(RemainingTime=@n5.ReturnValue, TotalDuration=@n5)
  ]}
 Set in Range State {nodes [
   n1 = Set in Range State!() -> (then, bPlayerInRange:bool=~, bEnvInRange:bool=~) {is=title K2Node_FunctionEntry}
   > n2 = GetDynamicMaterial(self=Get Reticle() -> (Reticle:Image Object Reference))
   > n3 = Set ReticleMaterialReference(ReticleMaterialReference:Material Instance Dynamic Object Reference=@n2) -> (Output_Get:Material Instance Dynamic Object Reference)
   > SetScalarParameterValue(self=@n3, ParameterName=FinalAlpha, Value=SelectFloat(A=1, B=0.6, bPickA=OR Boolean(A=@n1, B=@n1.bEnvInRange)))
   > SetScalarParameterValue(
    self=Get ReticleMaterialReference() -> (ReticleMaterialReference:Material Instance Dynamic Object Reference)
    ParameterName=IsEnemyInRange
    Value=SelectFloat(A=1, B=0, bPickA=Get bPlayerInRange() -> (bPlayerInRange:bool))
   )
   > n4 = Branch(Condition=Get bPlayerInRange())
   > Set Counter_NotInRange(Counter_NotInRange:double=0) -> (Output_Get:double)
   > n5 = Increment Float!(" ":Exec, Value:"Float (double-precision) (by ref)"=Get Counter() -> (Counter:double)) -> ("  ":Exec, Result:double=~) {is=title K2Node_MacroInstance}
   > n6 = Set Counter(Counter:double=@n5) -> (Output_Get:double)
   > Branch(Condition="float <= float"(A=@n6, B=1) -> (ReturnValue=~))
   n7 = Get Anim_EnemyInRange()
   > PlayAnimationForward(InAnimation=@n7)
   Set Counter(execute=@n4.else, Counter=0)
   > n8 = Increment Float(Value=Get Counter_NotInRange() -> (Counter_NotInRange:double)) -> (Result=~)
   > n9 = Set Counter_NotInRange(Counter_NotInRange=@n8)
   > Branch(Condition="float <= float"(A=@n9, B=1) -> (ReturnValue=~))
   > PlayAnimationReverse(InAnimation=@n7)
  ]}
 SetAmmoProgressBar {nodes [
   SetAmmoProgressBar() {is=title K2Node_FunctionEntry}
   > n1 = GetDynamicMaterial(self=Get Reticle())
   n2 = Get As B Demon Mask Fire()
   n3 = GetRemainingAmmo!(self:Fort Weapon Object Reference=@n2) -> (ReturnValue:int=0)
   n4 = Get TotalAmmoCount!(self:B Demon Mask Fire Object Reference=@n2) -> (TotalAmmoCount:Byte=0)
   > SetScalarParameterValue(
    self=@n1
    ParameterName=AnimateProgress
    Value=SafeDivide(A="To Float (Integer)"(InInt=@n3), B="To Float (Byte)"!(InByte:Byte=@n4) -> (ReturnValue:double))
   )
  ]}
]
variables [
 As B Demon Mask Fire {type=B_DemonMask_Fire_C, default=null}
 Cached Has Homing Target {type=boolean}
 Counter {type=number}
 Counter_NotInRange {type=number}
 Homing {type=WidgetAnimation, default=null}
 isNoAmmoShowing {type=boolean}
 ReticleMaterialReference {type=MaterialInstanceDynamic}
 ShowingHoming {type=boolean}
 Target State Component {type=FortWeaponComponent_TargetState, default=null}
]
widgets {
 info {
  inheritedWidgetCount=1
  parentClass=/Game/Athena/HUD/reticles/ReticleExtensionWidgets/ChargeStack/WBP_Reticle_RechargeStackBase.WBP_Reticle_RechargeStackBase_C
  rootWidgetClass=&Overlay
  widgetCount=6
 }
 widgets [
  ReticleOverlay {widgetClassPath=&Overlay, bIsVariable=true, widget=ReticleOverlay}
  Overlay_0 {
   widgetClassPath=&Overlay
   parent=ReticleOverlay
   widget=Overlay_0
   properties {renderTransform {angle=0, scale=<1.2 1.2>, shear=<0 0>, translation=<0 0>}}
   slot {class=OverlaySlot}
  }
  WBP_DemonMask_Fire_Reticle_Collection {
   widgetClassPath=&/WBP_DemonMask_Fire_Reticle_Collection
   bIsVariable=true
   parent=Overlay_0
   widget=WBP_DemonMask_Fire_Reticle_Collection
   properties {renderTransform {angle=0, scale=<1 1>, shear=<0 0>, translation=<0.6 15>}}
   slot {class=OverlaySlot}
  }
  Reticle {
   widgetClassPath=&Image
   bIsVariable=true
   parent=Overlay_0
   widget=Reticle
   properties {brush {
     bIsDynamicallyLoaded=false
     drawAs=Image
     imageType=FullColor
     mirroring=NoMirror
     resourceName=None
     resourceObject=/DemonMaskFire/UI/Reticle/Materials/M_UI_DemonMask_Fire
     tiling=NoTile
     imageSize=<128 128>
     margin=<ltrb 0 0 0 0>
     outlineSettings {
      bUseBrushTransparency=false
      roundingType=HalfHeightRadius
      width=0
      color {colorUseRule=UseColor_Specified, specifiedColor=<rgba 0 0 0 0>}
      cornerRadii=<0 0 0 0>
     }
     tintColor {colorUseRule=UseColor_Specified, specifiedColor=<rgba 1 1 1 1>}
     uVRegion {bIsValid=false, max=<0 0>, min=<0 0>}
    }}
   slot {class=OverlaySlot, properties {horizontalAlignment=HAlign_Center, verticalAlignment=VAlign_Center, padding=<ltrb 0 5 0 0>}}
  }
  WBP_Reticle_DemonMaskFire_AttachmentPoint {
   widgetClassPath=&/WBP_Reticle_DemonMaskFire_AttachmentPoint
   bIsVariable=true
   parent=ReticleOverlay
   widget=WBP_Reticle_DemonMaskFire_AttachmentPoint
   properties {renderTransform {angle=0, scale=<1.2 1.2>, shear=<0 0>, translation=<0 0>}}
   slot {class=OverlaySlot, properties {horizontalAlignment=HAlign_Center, verticalAlignment=VAlign_Center, padding=<ltrb 0 5 0 0>}}
  }
  TutorialHighlight {widgetClassPath=&TutorialHighlight}
 ]
}
